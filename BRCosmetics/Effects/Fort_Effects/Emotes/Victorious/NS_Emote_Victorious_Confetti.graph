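# Niagara system for the Victorious emote that fires a burst of multicolored confetti at a shared system-level random position.
# Four burst emitters layer small confetti sprites, a few mesh-oriented ribbon pieces, twinkling sparkles and a brief flare with a light.
# Drag, gravity and curl noise drive the falling pieces, with collision only on the main confetti.
# Front-end modifiers adjust confetti color and confetti and ribbon material parameters between lobby and gameplay.
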NiagaraSystem /BRCosmetics/Effects/Fort_Effects/Emotes/Victorious/NS_Emote_Victorious_Confetti
dependencies {
 usedDataInterfaces [
  &NiagaraDataInterfaceArrayFloat
  &NiagaraDataInterfaceCurve
  &NiagaraDataInterfaceSpriteRendererInfo
 ]
 usedDynamicInputs [
  &RotateVectorByQuaternion
  &self:ScratchDynamicInput
 ]
 usedModules [
  &SystemState
  &self:SystemSpawnScript.SystemScriptSource.SystemScriptGraph.NiagaraNodeAssignment_0.SetVariables_#1
 ]
}
emitters {
 Confetti {
  emitterUpdateScript [
   &EmitterState()
   &SpawnBurst_Instantaneous(
    Spawn Count=&RandomRangeInt(Minimum=8, Maximum=12, Evaluation Type=::NewEnumerator1)
    Use Spawn Probability=false
   )
  ]
  particleSpawnScript [
   /Niagara/Modules/Spawn/Initialization/InitializeParticle(
    Color=&HueShiftLinearColor(
     Hue=&RandomRangeFloat(Minimum=0, Maximum=1)
     Saturation=&FN_FrontEndFloatModifier(Gameplay=1, Front End=3)
     Value=2
     Linear Color=&MakeLinearColorFromVectorAndFloat(
      "Vector (RGB)"=&FN_FrontEndVector3Modifier(Gameplay=<0.511 0.5 0.5>, Front End=<0.45 0.43 0.43>)
      "Float (Alpha)"=1
     )
    )
    Lifetime=&RandomRangeFloat(Minimum=1, Maximum=3)
    Module.Write Color=true
    Module.Write Lifetime=true
    Module.Write Mass=true
    Module.Write Material Random=true
    Module.Write Position=false
    Module.Write Scale=false
    Module.Write SpriteRotation=true
    Module.Write SpriteSize=true
    Position=&SimulationPosition()
    Sprite Rotation=&RandomRangeFloat(Minimum=0, Maximum=360)
    Sprite Size=&Vector2DFromFloat(Value=&RandomRangeFloat(Minimum=3, Maximum=5))
   )
   &SphereLocation(
    MaskSpawnGroup=false
    Offset=$System.Random Position
    Override Seed=false
    Sphere Radius=&RandomRangeFloat(Minimum=10, Maximum=20)
    Surface Only=false
    UseNonUniformScale=false
    UseOffset=true
    UseSphereOrientation=false
   )
   &AddVelocityFromPoint(Module.Use Falloff Distance=false, Velocity Strength=&RandomRangeFloat(Minimum=50, Maximum=200))
   /SetVariables(Particles.Color=&MakeLinearColorFromVectorAndFloat(
     "Vector (RGB)"=&AddVectorToPosition(
      Position=HlslExpression() {hlslExpression=Particles.Color.xyz}
      Vector=&FN_FrontEndVector3Modifier(Gameplay=<0 0 0>, Front End=<0 0.1 0.25>)
     )
     "Float (Alpha)"=1
    )) {isSetParameters=true}
  ]
  particleUpdateScript [
   &ParticleState()
   &SpriteRotationRate(Rotation Rate=&Multiply_Float(
     A=&SelectFloatFromArray(
      Array Sampling Mode=::NewEnumerator0
      Direct Array Index=?
      Float Selection Array=DataInterface() {propertyValues {FloatData [-300,300]}}
      Random Seed=0
     )
     B=&RandomRangeFloat(Minimum=0.5, Maximum=1)
    ))
   &Drag(Drag=&RandomRangeFloat(Minimum=5, Maximum=10), Use Linear Drag=true, Use Rotational Drag=false)
   &GravityForce(Gravity=&RandomRangeVector(Minimum=<0 0 -360>, Maximum=<1 1 -200>))
   &CurlNoiseForce(Noise Strength=&RandomRangeFloat(Minimum=200, Maximum=300), Pan Noise=false)
   &Collision(
    Age Colliding Particles=false
    Friction=1111
    Maximum Penetration Correction Distance=-1.5
    Module.Control Rest State Via Surface Normal=false
    Module.EnableMaxCollisionCount=false
    Module.Randomize Collision Normal Vector=false
    Particle Radius Scale=0
   )
   &SolveForcesAndVelocity(Clamp Velocity=false, Limit Acceleration=false)
   &ScaleSpriteSize(Scale Factor=&Vector2DFromFloat(Value=&FloatFromCurve(FloatCurve=DataInterface() {propertyValues {Curve {keys [
          {arriveTangent=-0.4596, interpMode=RCIM_Cubic, leaveTangent=-0.4596, tangentMode=RCTM_User, time=0.55, value=1}
          {arriveTangent=-4.9031, interpMode=RCIM_Cubic, leaveTangent=-4.9031, tangentMode=RCTM_User, time=1, value=0}
         ]}}})))
   &DynamicMaterialParameters(
    Index 0 Param 1=&FloatFromCurve(FloatCurve=DataInterface() {propertyValues {Curve {keys [<at 0 1>,<at 1 0>]}}})
    Index 0 Param 3=1
    Index 0 Param 4=-1
    Index 1 Param 3=&FN_FrontEndFloatModifier(Gameplay=0, Front End=0.5)
    Index 1 Param 4=&FN_FrontEndFloatModifier(Gameplay=1, Front End=2.4)
    Module.Param0WriteEnabled=true
    Module.Param0WriteEnabled1=false
    Module.Param1WriteEnabled=false
    Module.Param1WriteEnabled1=false
    Module.Param2WriteEnabled=true
    Module.Param2WriteEnabled1=true
    Module.Param3WriteEnabled=true
    Module.Param3WriteEnabled1=true
    Write Parameter Index 1=true
   )
  ]
  rendererClasses [&NiagaraSpriteRendererProperties]
  renderers [{rendererClass=&NiagaraSpriteRendererProperties}]
 }
 Flare {
  emitterUpdateScript [
   &EmitterState()
   &SpawnBurst_Instantaneous(Spawn Count=1, Use Spawn Probability=false)
  ]
  particleSpawnScript [
   /Niagara/Modules/Spawn/Initialization/V2/InitializeParticle(
    AdjustAlpha=false
    AdjustHue=true
    AdjustSaturation=true
    AdjustValue=true
    Color=<rgba 1 0.2 0.2 1>
    Color Mode=::NewEnumerator3
    Hue Shift Range:Vector2f=<-1 1>
    Lifetime=0.15
    Position Mode=::NewEnumerator0
    Saturation Range:Vector2f=<0.8 0.9>
    Sprite Rotation Mode=::NewEnumerator2
    Sprite Size Mode=::NewEnumerator1
    Uniform Sprite Size Max=150
    Uniform Sprite Size Min=120
    Value Range:Vector2f=<1 1>
    Write Material Random=true
   )
   &SphereLocation(
    MaskSpawnGroup=false
    Offset=$System.Random Position
    Override Seed=false
    Sphere Radius=0
    Surface Only=false
    UseNonUniformScale=false
    UseOffset=true
    UseSphereOrientation=false
   )
  ]
  particleUpdateScript [
   &ParticleState()
   &ScaleSpriteSize(Scale Factor=&Vector2DFromFloat(Value=&FloatFromCurve(FloatCurve=DataInterface() {propertyValues {Curve {keys [
          {interpMode=RCIM_Cubic, tangentMode=RCTM_User, time=0, value=0}
          {interpMode=RCIM_Cubic, tangentMode=RCTM_User, time=0.4, value=1}
          {tangentMode=RCTM_User, time=1, value=0}
         ]}}})))
   &ScaleColor(
    Scale RGB=&VectorFromFloat(Value=&FloatFromCurve(
      FloatCurve=DataInterface() {propertyValues {Curve {keys [{interpMode=RCIM_Cubic, tangentMode=RCTM_User, time=0, value=1}, {interpMode=RCIM_Cubic, tangentMode=RCTM_User, time=1, value=0}]}}}
      Scale Curve=5
     ))
    ScaleA=true
    ScaleRGB=true
   )
  ]
  rendererClasses [&NiagaraSpriteRendererProperties, &NiagaraLightRendererProperties]
  renderers [
   {rendererClass=&NiagaraSpriteRendererProperties}
   {rendererClass=&NiagaraLightRendererProperties, rendererIndex=1}
  ]
 }
 RibbonConfetti {
  emitterUpdateScript [
   &EmitterState()
   &SpawnBurst_Instantaneous(
    Spawn Count=&RandomRangeInt(Minimum=0, Maximum=3, Evaluation Type=::NewEnumerator1)
    Use Spawn Probability=false
   )
  ]
  particleSpawnScript [
   /Niagara/Modules/Spawn/Initialization/InitializeParticle(
    Color=&HueShiftLinearColor(Hue=&RandomRangeFloat(Minimum=0, Maximum=1), Saturation=1, Value=1, Linear Color=<rgba 0.5 0.4 0.4 1>)
    Lifetime=&RandomRangeFloat(Minimum=0.5, Maximum=2.5)
    Module.Write Color=true
    Module.Write Lifetime=true
    Module.Write Mass=true
    Module.Write Material Random=true
    Module.Write Position=false
    Module.Write Scale=false
    Module.Write SpriteRotation=true
    Module.Write SpriteSize=true
    Position=&SimulationPosition()
    Sprite Rotation=&RandomRangeFloat(Minimum=0, Maximum=360)
    Sprite Size=&RandomRangeVector2D(Minimum=<4 10>, Maximum=<5 12.5>)
   )
   &SphereLocation(
    MaskSpawnGroup=false
    Offset=$System.Random Position
    Override Seed=false
    Sphere Radius=&RandomRangeFloat(Minimum=5, Maximum=20)
    Surface Only=false
    UseNonUniformScale=false
    UseOffset=true
    UseSphereOrientation=false
   )
   &AddVelocityFromPoint(Module.Use Falloff Distance=false, Velocity Strength=&RandomRangeFloat(Minimum=25, Maximum=150))
   &SubUVAnimation(
    End Frame Range Override=3
    Start Frame Range Override=3
    SubUV Animation Mode:ENiagara_SubUVLookupModeV2=::NewEnumerator2
    UseEndFrame=true
    UseStartFrame=true
   )
  ]
  particleUpdateScript [
   &ParticleState()
   &Drag(Use Linear Drag=true, Use Rotational Drag=false)
   &SpriteRotationRate(Rotation Rate=&Multiply_Float(
     A=&SelectFloatFromArray(
      Array Sampling Mode=::NewEnumerator0
      Direct Array Index=?
      Float Selection Array=DataInterface() {propertyValues {FloatData [-400,400]}}
      Random Seed=0
     )
     B=&RandomRangeFloat(Minimum=0.3, Maximum=1)
    ))
   &AerodynamicDrag(Use Linear Drag=true, Use Rotational Drag=true)
   &CurlNoiseForce(Noise Strength=&RandomRangeFloat(Minimum=200, Maximum=300), Pan Noise=false)
   &GravityForce(Gravity=&RandomRangeVector(Minimum=<0 0 -360>, Maximum=<0 0 -500>))
   &SolveForcesAndVelocity(Clamp Velocity=false, Limit Acceleration=false)
   &ScaleSpriteSize(Scale Factor=&Vector2DFromFloat(Value=&FloatFromCurve(FloatCurve=DataInterface() {propertyValues {Curve {keys [
          {arriveTangent=-0.4596, interpMode=RCIM_Cubic, leaveTangent=-0.4596, tangentMode=RCTM_User, time=0.55, value=1}
          {arriveTangent=-4.0966, interpMode=RCIM_Cubic, leaveTangent=-4.0966, tangentMode=RCTM_User, time=1, value=0}
         ]}}})))
   &AlignSpriteToMeshOrientation(SetAlignment=true, SetFacing=true)
   &DynamicMaterialParameters(
    Index 1 Param 3=&FN_FrontEndFloatModifier(Gameplay=0, Front End=1)
    Index 1 Param 4=&FN_FrontEndFloatModifier(Gameplay=1, Front End=1)
    Module.Param0WriteEnabled1=false
    Module.Param1WriteEnabled1=false
    Module.Param2WriteEnabled1=true
    Module.Param3WriteEnabled1=true
    Write Parameter Index 0=false
    Write Parameter Index 1=true
   )
  ]
  rendererClasses [&NiagaraSpriteRendererProperties]
  renderers [{rendererClass=&NiagaraSpriteRendererProperties}]
 }
 Sparkles {
  emitterUpdateScript [
   &EmitterState()
   &SpawnBurst_Instantaneous(Spawn Count=30, Use Spawn Probability=false)
  ]
  particleSpawnScript [
   &InitializeParticle(
    Color=&HueShiftLinearColor(Hue=&RandomRangeFloat(Minimum=0, Maximum=1), Saturation=1, Value=5, Linear Color=<rgba 1.5 0.5 0.5 1>)
    Lifetime=&Power(a=&RandomRangeFloat(Minimum=0.46, Maximum=1.5), b=2)
    Module.Write Color=true
    Module.Write Lifetime=true
    Module.Write Mass=true
    Module.Write Material Random=true
    Module.Write Position=false
    Module.Write Scale=false
    Module.Write SpriteRotation=true
    Module.Write SpriteSize=true
    Position=&SimulationPosition()
    Sprite Rotation=&RandomRangeFloat(Minimum=0, Maximum=360)
    Sprite Size=&Vector2DFromFloat(Value=&RandomRangeFloat(Minimum=3, Maximum=6))
   )
   &SphereLocation(
    MaskSpawnGroup=false
    Offset=$System.Random Position
    Override Seed=false
    Sphere Radius=&RandomRangeFloat(Minimum=2, Maximum=6)
    Surface Only=false
    UseNonUniformScale=false
    UseOffset=true
    UseSphereOrientation=false
   )
   &AddVelocityFromPoint(Module.Use Falloff Distance=false, Velocity Strength=&RandomRangeFloat(Minimum=50, Maximum=150))
  ]
  particleUpdateScript [
   &ParticleState()
   &Drag(Drag=2, Use Linear Drag=true, Use Rotational Drag=false)
   &GravityForce(Gravity=&RandomRangeVector(Minimum=<0 0 -100>, Maximum=<1 1 -300>))
   &CurlNoiseForce(Noise Strength=&RandomRangeFloat(Minimum=200, Maximum=300), Pan Noise=false)
   &SolveForcesAndVelocity(Clamp Velocity=false, Limit Acceleration=false)
   &ScaleColor(Scale Alpha=&Lerp_Float(A=1, B=0, Alpha=$Particles.NormalizedAge), ScaleA=true, ScaleRGB=false)
   &ScaleSpriteSize(Scale Factor=&Vector2DFromFloat(Value=&FloatFromCurve(FloatCurve=DataInterface() {propertyValues {Curve {keys [<at 0.0018 0.1957>,<at 0.0527 0.819>,<at 0.1169 0.6555>,<at 0.1748 0.9015>,<at 0.2394 0.582>,<at 0.332 1.0023>,<at 0.4947 0.3707>,<at 0.6727 1>,<at 1 0>]}}})))
  ]
  rendererClasses [&NiagaraSpriteRendererProperties]
  renderers [{rendererClass=&NiagaraSpriteRendererProperties}]
 }
}
summary {
 systemName=NS_Emote_Victorious_Confetti
 emitters [
  {emitterName=Confetti, rendererClasses [&NiagaraSpriteRendererProperties]}
  {emitterName=RibbonConfetti, rendererClasses [&NiagaraSpriteRendererProperties]}
  {emitterName=Sparkles, rendererClasses [&NiagaraSpriteRendererProperties]}
  {emitterName=Flare, rendererClasses [&NiagaraSpriteRendererProperties, &NiagaraLightRendererProperties]}
 ]
 userVariables [
  User.bIsFrontEnd {type=NiagaraBool}
  User.bIsFrontEndPreview {type=NiagaraBool}
  User.RibbonIDOffset {type=NiagaraInt32}
  User.SkeletalMesh {
   type=NiagaraDataInterfaceSkeletalMesh
   defaultValue {
    struct=NiagaraExt_VariableValue_DataInterface
    value {dataInterface=&self:NiagaraDataInterfaceSkeletalMesh_0}
   }
  }
 ]
}
